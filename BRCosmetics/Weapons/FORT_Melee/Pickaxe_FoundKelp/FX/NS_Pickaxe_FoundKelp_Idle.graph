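# Niagara idle effect for the FoundKelp pickaxe, drawing glowing blade sprites attached to the weapon's sockets.
# One Blades emitter bursts groups of three particles that follow socket-sampled positions, with per-group offsets, and persist past their lifetime.
# Curves drive dynamic material parameters, while a waveform flickers the tip-colored glow alpha across three sprite renderers.
# User.Variant switches sprite orientation, and front-end modifiers adjust fade angles, glow strength and side-blade size.
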NiagaraSystem /BRCosmetics/Weapons/FORT_Melee/Pickaxe_FoundKelp/FX/NS_Pickaxe_FoundKelp_Idle
dependencies {
 usedDataInterfaces [
  &NiagaraDataInterfaceArrayFloat2
  &NiagaraDataInterfaceArrayFloat3
  &NiagaraDataInterfaceCurve
  &NiagaraDataInterfaceSocketReader
 ]
 usedDynamicInputs [&SelectVector2DFromArray, &MergeHelper_ColorSwitch]
 usedModules [
  &self:SystemSpawnScript.SystemScriptSource.SystemScriptGraph.NiagaraNodeAssignment_0.SetVariables_#1
  &self:SystemSpawnScript.SystemScriptSource.SystemScriptGraph.NiagaraNodeAssignment_2.SetVariables_#2
  &SystemState
  &self:SystemSpawnScript.SystemScriptSource.SystemScriptGraph.NiagaraNodeAssignment_3.SetVariables_#3
  &self:SystemSpawnScript.SystemScriptSource.SystemScriptGraph.NiagaraNodeAssignment_4.SetVariables_#4
  &self:SystemSpawnScript.SystemScriptSource.SystemScriptGraph.NiagaraNodeAssignment_6.SetVariables_#5
 ]
}
emitters {Blades {
  emitterSpawnScript [/SetVariables(
    Emitter.BladeSideViewAngleFadeMin:NiagaraFloat=&FN_FrontEndFloatModifier(Gameplay=0.9, Front End=0.8)
    Emitter.GlowViewAngleFadeMin:NiagaraFloat=&FN_FrontEndFloatModifier(Gameplay=0.9, Front End=0.5)
   ) {isSetParameters=true}]
  emitterUpdateScript [
   &EmitterState(Life Cycle Mode=::NewEnumerator1, Loop Behavior=::NewEnumerator1, Loop Delay=0.1, Loop Duration Mode=::NewEnumerator1, UseLoopDelay=true)
   &SpawnBurst_Instantaneous(Spawn Count=3, Use Spawn Probability=false, UseLoopCountLimit=false)
   &SpawnBurst_Instantaneous(Spawn Count=3, Spawn Time=1.0833, Use Spawn Probability=false, UseLoopCountLimit=false) {module=SpawnBurst_Instantaneous001}
  ]
  particleSpawnScript [
   &FN_SetSpawnGroup(Number of Spawn Groups=3)
   &InitializeParticle(
    Lifetime=0.65
    Position Mode=::NewEnumerator0
    Sprite Rotation Mode=::NewEnumerator0
    Sprite Size Mode=::NewEnumerator3
    Uniform Sprite Size=60
    Write Material Random=true
   )
   &SpriteFacingAndAlignment(
    Sprite Alignment=<0 0 1>
    Sprite Facing=&MergeHelper_VectorSwitch(
     0=<0 1 0>
     1=<0 1 0>
     2=<0 1 -0.07>
     3=?
     Index=&MakeIntFromFloat(Float=$User.Variant, Conversion Mode=::NewEnumerator1)
     Num=::NewEnumerator1
    )
    WriteAlignment=true
    WriteFacing=true
   )
  ]
  particleUpdateScript [
   &ParticleState(Kill Particles When Lifetime Has Elapsed=false, Let Infinitely Lived Particles Die When Emitter Deactivates=true)
   &SocketLocation(
    Position Sampling=::NewEnumerator1
    SamplePosOffset=true
    Sampled Position Offset=&AddVectorToPosition(
     Position=&SelectVectorFromArray(
      Direct Array Index=&SpawnGroup()
      Vector Selection Array=DataInterface() {propertyValues {FloatData [<0 0 0>, <0 -0.15 0>, <0 0.15 0>]}}
     )
     Vector=<0 0 0>
    )
    Socket Reader=$System.Socket Reader
    Socket Sampling Mode=::NewEnumerator3
    VelocitySpeedLimit=false
   )
   &SocketLocation(
    Position Sampling=::NewEnumerator1
    SamplePosOffset=false
    Socket Reader=$System.Socket Reader
    Socket Sampling Mode=::NewEnumerator3
    VelocitySpeedLimit=false
   ) {module=SocketLocation001}
   /SetVariables(
    Particles.BladePosition:NiagaraPosition=&AddVectorToPosition(
     Position=$Output.SocketLocation.SampledPosition
     Vector=&MergeHelper_VectorSwitch(
      0=<0 0 0>
      1=<0 0 0>
      2=<0 0 0>
      3=?
      Index=&MakeIntFromFloat(Float=$User.Variant, Conversion Mode=::NewEnumerator1)
      Num=::NewEnumerator1
     )
    )
    Particles.GlowPosition:NiagaraPosition=&AddVectorToPosition(
     Position=$Output.SocketLocation001.SampledPosition
     Vector=&MergeHelper_VectorSwitch(
      0=<0 0 0>
      1=<0 0 0>
      2=<0 0 0>
      3=?
      Index=&MakeIntFromFloat(Float=$User.Variant, Conversion Mode=::NewEnumerator1)
      Num=::NewEnumerator1
     )
    )
   ) {isSetParameters=true}
   &DynamicMaterialParameters(
    Index 0 Param 1=&FloatFromCurve(FloatCurve=DataInterface() {propertyValues {Curve {keys [<at 0.25 0>,<at 1 1>]}}})
    Index 0 Param 2=0.5
    Index 0 Param 3=&FloatFromCurve(FloatCurve=DataInterface() {propertyValues {Curve {keys [<at 0.15 0>,<at 0.5 1>]}}})
    Index 0 Param 4=&FloatFromCurve(FloatCurve=DataInterface() {propertyValues {Curve {keys [<at 0 0>,<at 0.25 1>]}}})
    Index 1 Param 1=&FloatFromCurve(FloatCurve=DataInterface() {propertyValues {Curve {keys [<at 0.4 0>,<at 0.8 1>]}}})
    Module.Index0Float0Write=true
    Module.Index0Float1Write=true
    Module.Index0Float2Write=true
    Module.Index0Float3Write=true
    Module.Index1Float0Write=true
    Module.Index1Float1Write=true
    Module.Index1Float2Write=true
    Module.Index1Float3Write=true
    Write Parameter Index 1=::NewEnumerator1
   )
   /SetVariables(
    Particles.GlowColor=&MultiplyLinearColor(
     Linear Color=$System.Main Color Tip
     Scale Factor=&MakeVector4FromVectorAndFloat(
      "Vector (RGB)"=<1 1 1>
      "Float (Alpha)"=&Multiply_Float(
       A=&FloatFromCurve(
        FloatCurve=DataInterface() {propertyValues {Curve {defaultValue=1, keys [<at 0.75 0>, {tangentMode=RCTM_User, time=1, value=1}]}}}
        CurveIndex=$Particles.Age
        Scale Curve=&Waveform(
         Global Clamp Min=0.3
         Global Clamp Max=0.5
         "Waveform [1]"=::NewEnumerator2
         "[1] Frequency"=2550
         "[1] Compound Frequency"=2000
         "[1] Amplitude Min/Max"=<0 1>
         "Waveform [2] Blend"=::NewEnumerator2
        )
       )
       B=&FN_FrontEndFloatModifier(Gameplay=1, Front End=0.15)
      )
     )
    )
    Particles.GlowRotation=90
    Particles.GlowSize=&Multiply_Vector2D_ByFloat(Float=1, Vector2D=<62.5 3>)
    Particles.GlowSpriteAlignment:Vector3f=&MergeHelper_VectorSwitch(
     0=<0 0 1>
     1=<0 0 1>
     2=<0 0.07 1>
     3=?
     Index=&MakeIntFromFloat(Float=$User.Variant, Conversion Mode=::NewEnumerator1)
     Num=::NewEnumerator1
    )
    Particles.GlowSpriteFacing:Vector3f=<1 0 0>
   ) {isSetParameters=true}
   /SetVariables(
    Particles.SideBladeColor:LinearColor=&Multiply_LinearColorByFloat(Linear Color=$System.Main Color Tip, Float=1)
    Particles.SideBladeSize:Vector2f=&Multiply_Vector2D(
     A=$Particles.GlowSize
     B=&FN_FrontEndVector2Modifier(Front End:Vector2f=<0.75 0.25>, Gameplay:Vector2f=<1 0.25>)
    )
   ) {isSetParameters=true}
  ]
  rendererClasses [&NiagaraSpriteRendererProperties]
  renderers [
   {rendererClass=&NiagaraSpriteRendererProperties}
   {rendererClass=&NiagaraSpriteRendererProperties, rendererIndex=1}
   {rendererClass=&NiagaraSpriteRendererProperties, rendererIndex=2}
  ]
 }}
summary {
 systemName=NS_Pickaxe_FoundKelp_Idle
 emitters [{emitterName=Blades, rendererClasses [&NiagaraSpriteRendererProperties]}]
 userVariables [
  User.bIsFrontEnd {type=NiagaraBool}
  User.bIsFrontEndPreview {type=NiagaraBool}
  User.Variant {type=NiagaraFloat, defaultValue=0, description=""}
  User.VariantSwitch {type=NiagaraFloat, defaultValue=0, description=""}
 ]
}
